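# Translucent unlit, two-sided Niagara mesh-particle material for a floppy skydiving contrail tube with scrolling, fire-like streaks.
# Layered panning noise, a gradient and a chunk mask form one mask, driving a three-color emissive ramp and opacity, faded by particle alpha, vertex color, camera proximity and view angle.
# WPO combines a dynamic-parameter-driven bend weighted along V, noise and gradient displacement, and vector-noise wobble, which is disabled on forward and mobile paths.
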Material /BRCosmetics/Athena/Effects/Particles/Skydiving/NS_Contrail_Tube/M_Floppy_Tube
graph {
 properties {
  bUsedWithNiagaraMeshParticles=true
  blendMode=BLEND_Translucent
  pixelDepthOffsetMode=PDOM_Legacy
  shadingModel=MSM_Unlit
  twoSided=true
  usedShadingModels=Unlit
 }
 nodes [
  n1 = DynamicParameter() {defaultValue=<rgba 0 0 0 1>, paramNames [X, Y, Z, Bend WPO Intensity]}
  n2 = Subtract(
   A=MaterialFunctionCall(X=@n1.X, Y=@n1.X, Z=@n1.X) {materialFunction=&MakeFloat3}.Result
   B=ParticlePositionWS()
  )
  n3 = ComponentMask(None=TextureCoordinate()) {g=true}
  n4 = Multiply(A=Power(Base=OneMinus(None=@n3), Exp=ScalarParameter() {defaultValue=2.5, group=WPO, parameterName=WPOBendPower}), B=@n1.Bend WPO Intensity)
  n5 = TextureCoordinate() {vTiling=0.5}
  n6 = ComponentMask(None=VectorParameter() {group=Fire, parameterName=UVScale, defaultValue=<rgba 1 0.25 0 0>}.RGB) {g=true, r=true}
  n7 = DynamicParameter() {parameterIndex=1, defaultValue=<rgba 1 1 0 1>, paramNames [Fire Speed Mult, Mask Speed V Mult, ------------, ------------]}
  n8 = Multiply(A=Multiply(A=Time(), B=ScalarParameter() {defaultValue=2.5, group=Fire, parameterName=Fire Speed}), B=@n7.Fire Speed Mult)
  n9 = TextureSample(UVs=Panner(Coordinate=Multiply(A=@n5, B=@n6), Time=@n8) {speedX=-0.05, speedY=1}) {texture=&T_NoisePack_01}
  n10 = Constant() {r=0.2}
  n11 = Panner(Coordinate=Multiply(A=@n6, B=TextureCoordinate() {uTiling=4}), Time=@n8) {speedX=0.1, speedY=1}
  n12 = Multiply(A=@n6, B=TextureCoordinate() {uTiling=12, vTiling=3})
  n13 = Multiply(
   A=Add(
    A=MaterialFunctionCall(In=@n9.G, Contrast=@n10) {materialFunction=&CheapContrast}.Result
    B=TextureSample(UVs=@n11) {texture=&T_NoisePack_01}.G
   )
   B=TextureSample(UVs=Panner(Coordinate=@n12, Time=@n8) {speedX=-0.1, speedY=1}) {texture=&T_NoisePack_01}.G
  )
  n14 = ScalarParameter() {defaultValue=5, group=WPO, parameterName=Distort}
  n15 = Multiply(A=Multiply(A=VertexNormalWS(), B=@n13), B=@n14)
  n16 = TextureCoordinate()
  n17 = Multiply(
   A=TextureSample(UVs=Panner(Coordinate=@n16, Time=@n8) {speedY=0.25}) {texture=/Game/Global/Textures/Default/Masks_and_Gradients/CylGradient}.RGB
   B=@n14
  ) {constB=35}
  n18 = Subtract(
   A=Subtract(A=WorldPosition().XYZ, B=ObjectPositionWS())
   B=Multiply(A=ScalarParameter() {defaultValue=150, group=WPO, parameterName=Curl_Speed}, B=Time())
  )
  n19 = ScalarParameter() {defaultValue=150, group=WPO, parameterName=Curl_Scale}
  n20 = Multiply(
   A=ScalarParameter() {defaultValue=15, group=WPO, parameterName=Curl_Intensity}
   B=VectorNoise(World Position=Divide(A=@n18, B=@n19)) {noiseFunction=VNF_VectorALU}
  )
  n21 = Constant()
  Add(
   A=StaticSwitchParameter(True=Multiply(A=@n2, B=@n4), False=Constant()) {defaultValue=true, group=WPO, parameterName=WPO From Velocity}
   B=Add(A=Add(A=@n15, B=@n17), B=ShadingPathSwitch(Default=@n20, Forward=@n21, Mobile=@n21))
  )
  n22 = TextureCoordinate()
  n23 = AppendVector(
   A=Multiply(A=ComponentMask(None=@n22) {r=true}, B=ScalarParameter() {defaultValue=1, group=Alpha, parameterName=Alpha1 Tile U})
   B=Multiply(A=ComponentMask(None=@n22) {g=true}, B=ScalarParameter() {defaultValue=0.001, group=Alpha, parameterName=Alpha1 Tile V})
  )
  n24 = DynamicParameter() {parameterIndex=1, defaultValue=<rgba 1 1 0 1>, paramNames [Fire Speed Mult, Mask Speed V Mult, ------------, ------------]}
  n25 = Multiply(A=Multiply(A=Time(), B=ScalarParameter() {defaultValue=0.05, group=Alpha, parameterName=Alpha1 Pan Speed V}), B=@n24.Mask Speed V Mult)
  n26 = DynamicParameter() {parameterIndex=1, defaultValue=<rgba 1 1 0 1>, paramNames [Fire Speed Mult, Mask Speed V Mult, ------------, ------------]}
  n27 = AppendVector(
   A=ScalarParameter() {defaultValue=0.01, group=Alpha, parameterName=Alpha1 Pan Speed U}
   B=Multiply(A=ScalarParameter() {defaultValue=0.05, group=Alpha, parameterName=Alpha1 Pan Speed V}, B=@n26.Mask Speed V Mult)
  )
  n28 = Add(
   A=@n13
   B=Power(
    Base=MaterialFunctionCall() {materialFunction=&LinearGradient}.VGradient
    Exp=ScalarParameter() {defaultValue=3, group=Fire, parameterName=FireLine}
   )
  )
  n29 = Multiply(
   A=TextureSample(UVs=Panner(Coordinate=@n23, Time=@n25, Speed=@n27) {speedX=-1, speedY=0.1}) {texture=/Game/Global/Textures/Default/Masks_and_Gradients/chunkma_2}.RGB
   B=MaterialFunctionCall(In=Power(Base=@n28) {constExponent=0.5}, Contrast=ScalarParameter() {group=Fire, parameterName=HardEdgeness}) {materialFunction=&CheapContrast}.Result
  )
  MaterialFunctionCall(Emissive=MaterialFunctionCall(
    A=VectorParameter() {parameterName=Color3, defaultValue=<rgba 0.7793 0 1 1>}.RGB
    B=VectorParameter() {parameterName=Color2, defaultValue=<rgba 0.0124 0 1 1>}.RGB
    C=VectorParameter() {parameterName=Color1, defaultValue=<rgba 0.0278 1 0.8498 1>}.RGB
    Alpha=Clamp(None=@n29)
   ) {materialFunction=&Lerp_3Color}.Result) {materialFunction=&MF_FX_EmissiveAdjustments}
  n30 = Multiply(A=ParticleColor().A, B=VertexColor())
  n31 = MaterialFunctionCall(Distance=Constant() {r=250}) {materialFunction=&Near_Camera_Fade}
  n32 = Abs(None=DotProduct(A=CameraVectorWS(), B=VertexNormalWS()))
  n33 = ScalarParameter() {defaultValue=1, desc=Exponent for frensel falloff., group="4) Fade", parameterName=Frenel Fade Exponent}
  Multiply(A=@n29, B=Multiply(A=Multiply(A=@n30, B=@n31.Result), B=Power(Base=@n32, Exp=@n33)))
  Time() {period=1}
 ]
}
