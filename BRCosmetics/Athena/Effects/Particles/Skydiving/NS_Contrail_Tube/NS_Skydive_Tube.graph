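# Skydiving contrail cosmetic that builds a glowing tube effect around the player, with mesh, light, ribbon, smoke and speed-line layers.
# Tube, bowl and glow meshes attach to the skeletal mesh; the tube's material receives smoothed world position and integrator-driven parameters.
# Transparent parent particles form rings around the owner, feeding two ribbon trail emitters through particle attribute readers.
# A front-end preview flag adjusts tube rendering and adds vertical motion to smoke and ribbons.
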NiagaraSystem /BRCosmetics/Athena/Effects/Particles/Skydiving/NS_Contrail_Tube/NS_Skydive_Tube
dependencies {
 usedDataInterfaces [&NiagaraDataInterfaceCurve, &NiagaraDataInterfaceParticleRead, &NiagaraDataInterfaceColorCurve]
 usedDynamicInputs [&FN_FrontEndFloatModifier, &SetBoolByFloatComparison]
 usedModules [
  &SystemState
  /Game/Effects/Niagara/Modules/Skydiving/FN_DivingAttributes/v4/FN_DivingAttributes_v4
  &self:SystemSpawnScript.SystemScriptSource.SystemScriptGraph.NiagaraNodeAssignment_0.SetVariables_#1
  &FN_Integrator
 ]
}
emitters {
 Bowl_Multi {
  emitterSpawnScript [&FN_PreserveSkydiveUserVariables() {module=PreserveSkydiveUserVariables}]
  emitterUpdateScript [
   &EmitterState()
   &SpawnBurst_Instantaneous(Spawn Count=1, Use Spawn Probability=false) {enabled=false}
   &SpawnRate(SpawnRate=2, Use Spawn Probability=false)
  ]
  particleSpawnScript [
   &FN_SetSpawnGroup(Number of Spawn Groups=1) {enabled=false, module=FN_SetSpawnGroup001}
   /Niagara/Modules/Spawn/Initialization/InitializeParticle(
    Color=<rgba 3 3 3 1>
    Mesh Scale=<1 1.5 1.5>
    Module.Write Color=true
    Module.Write Lifetime=true
    Module.Write Mass=false
    Module.Write Material Random=true
    Module.Write Position=false
    Module.Write Scale=true
    Module.Write SpriteRotation=false
    Module.Write SpriteSize=false
    Position=&SimulationPosition()
    Sprite Rotation=&UniformRangedFloat(Minimum=0, Maximum=1)
   )
   &SkeletalMeshLocation(
    Bone / Socket Index=4
    Bone Sampling Mode=::NewEnumerator4
    Inherit Velocity=false
    Interpolate Mesh Sampling=false
    MaskSpawnGroup=false
    Sampled Position Offset=<0 0 15>
    Skeletal Mesh=$User.SkeletalMesh
   )
   &InitialMeshOrientation(
    Orientation Axis=<0 0 1>
    Rotation=&MakeVector(X=&RandomRangeFloat(Minimum=-1, Maximum=1), Y=0, Z=0)
    Use Orientation Vector=true
    Use Rotation Vector=true
   )
  ]
  particleUpdateScript [
   /Game/Effects/Niagara/Modules/Skydiving/FN_DivingAttributes/FN_DivingAttributes_v2(Accumulate Dive=0.1, Accumulate Rest=0.1, SkeletalMesh:NiagaraDataInterfaceSkeletalMesh=$User.SkeletalMesh) {module=FN_DivingAttributes_v2001}
   &ParticleState()
   &SolveForcesAndVelocity(Clamp Velocity=false, Limit Acceleration=false)
   &ScaleColor(
    Scale Alpha=&FloatFromCurve(FloatCurve=DataInterface() {propertyValues {Curve {keys [
         {arriveTangent=4.0516, interpMode=RCIM_Cubic, leaveTangent=4.0516, tangentMode=RCTM_User, time=0, value=0}
         {arriveTangent=0.0961, interpMode=RCIM_Cubic, leaveTangent=0.0961, tangentMode=RCTM_User, time=0.25, value=0.5}
         {arriveTangent=0.2272, interpMode=RCIM_Cubic, leaveTangent=0.2272, tangentMode=RCTM_User, time=1, value=0}
        ]}}})
    Scale RGB=&MakeVectorFromLinearColorRGB(Color=&ColorFromCurve(ColorCurve=DataInterface() {propertyValues {
        AlphaCurve {keys [<at 0 1 RCIM_Cubic>]}
        BlueCurve {keys [<at 0 0.4227 RCIM_Cubic>,<at 1 1 RCIM_Cubic>]}
        GreenCurve {keys [<at 0 1 RCIM_Cubic>,<at 1 0 RCIM_Cubic>]}
        RedCurve {keys [<at 0 0.0325 RCIM_Cubic>,<at 1 0.7358 RCIM_Cubic>]}
       }}))
    ScaleA=true
    ScaleRGB=true
   )
   &DynamicMaterialParameters(
    Index 0 Param 1=1
    Index 0 Param 2=1
    Index 0 Param 3=&FloatFromCurve(FloatCurve=DataInterface() {propertyValues {Curve {keys [{arriveTangent=1.3962, interpMode=RCIM_Cubic, leaveTangent=1.3962, tangentMode=RCTM_User, time=0.0032, value=0.136}, <at 1 1>]}}})
    Index 0 Param 4=&FloatFromCurve(
     FloatCurve=DataInterface() {propertyValues {Curve {keys [{arriveTangent=-1.0048, interpMode=RCIM_Cubic, leaveTangent=-1.0048, tangentMode=RCTM_User, time=-0.0046, value=1.0113}, <at 0.7 0.3>]}}}
     Scale Curve=0.5
    )
    Module.Param0WriteEnabled=false
    Module.Param1WriteEnabled=true
    Module.Param2WriteEnabled=true
    Module.Param3WriteEnabled=true
   )
   &ScaleMeshSize(Scale Factor=&VectorFromFloat(Value=&FloatFromCurve(FloatCurve=DataInterface() {propertyValues {Curve {keys [{arriveTangent=1.9256, interpMode=RCIM_Cubic, leaveTangent=1.9256, tangentMode=RCTM_User, time=0, value=0}, <at 1 1 RCIM_Cubic>]}}})))
  ]
  rendererClasses [&NiagaraMeshRendererProperties]
  renderers [{rendererClass=&NiagaraMeshRendererProperties, rendererIndex=0}]
 }
 CenterGlow {
  emitterSpawnScript [&FN_PreserveSkydiveUserVariables() {module=PreserveSkydiveUserVariables}]
  emitterUpdateScript [
   &EmitterState(
    Life Cycle Mode=::NewEnumerator1
    Loop Behavior=::NewEnumerator1
    Loop Duration Mode=::NewEnumerator1
    Reset Age On Awaken=true
    Scalability Mode=::NewEnumerator1
    UseLoopDelay=false
   )
   &SpawnBurst_Instantaneous(Spawn Count=1, Use Spawn Probability=false)
  ]
  particleSpawnScript [
   &FN_SetSpawnGroup(Number of Spawn Groups=1) {module=FN_SetSpawnGroup001}
   &InitializeParticle(
    Color=<rgba 0.451 1.5053 3 1>
    Lifetime=2
    Mesh Scale=<0.35 1.5 1.5>
    Module.Write Color=true
    Module.Write Lifetime=true
    Module.Write Mass=false
    Module.Write Material Random=true
    Module.Write Position=false
    Module.Write Scale=true
    Module.Write SpriteRotation=false
    Module.Write SpriteSize=false
    Position=&SimulationPosition()
    Sprite Rotation=&UniformRangedFloat(Minimum=0, Maximum=1)
   )
   &SkeletalMeshLocation(
    Bone / Socket Index=4
    Bone Sampling Mode=::NewEnumerator4
    Inherit Velocity=false
    Interpolate Mesh Sampling=false
    MaskSpawnGroup=false
    Sampled Position Offset=<0 0 35>
    Skeletal Mesh=$User.SkeletalMesh
   )
   &InitialMeshOrientation(Orientation Axis=<0 0 1>, Use Orientation Vector=true, Use Rotation Vector=false)
  ]
  particleUpdateScript [
   &FN_DivingAttributes_v2(Accumulate Dive=0.1, Accumulate Rest=0.1, SkeletalMesh:NiagaraDataInterfaceSkeletalMesh=$User.SkeletalMesh) {module=FN_DivingAttributes_v2001}
   &ParticleState(Kill Particles When Lifetime Has Elapsed=false, Let Infinitely Lived Particles Die When Emitter Deactivates=true)
   &SolveForcesAndVelocity(Clamp Velocity=false, Limit Acceleration=false)
   &ScaleColor(
    Scale Alpha=&FloatFromCurve(FloatCurve=DataInterface() {propertyValues {Curve {keys [
         {arriveTangent=4.0516, interpMode=RCIM_Cubic, leaveTangent=4.0516, tangentMode=RCTM_User, time=0, value=0}
         {arriveTangent=0.0961, interpMode=RCIM_Cubic, leaveTangent=0.0961, tangentMode=RCTM_User, time=0.25, value=1}
         {arriveTangent=0.2272, interpMode=RCIM_Cubic, leaveTangent=0.2272, tangentMode=RCTM_User, time=1, value=1}
        ]}}})
    ScaleA=true
    ScaleRGB=false
   )
  ]
  rendererClasses [&NiagaraMeshRendererProperties]
  renderers [{rendererClass=&NiagaraMeshRendererProperties, rendererIndex=0}]
 }
 Parent {
  emitterUpdateScript [
   &EmitterState(
    Life Cycle Mode=::NewEnumerator1
    Loop Behavior=::NewEnumerator1
    Loop Duration Mode=::NewEnumerator1
    Reset Age On Awaken=true
    Scalability Mode=::NewEnumerator1
    UseLoopDelay=false
   )
   &SpawnBurst_Instantaneous(Spawn Count=7, Use Spawn Probability=false)
  ]
  particleSpawnScript [
   /Niagara/Modules/Spawn/Initialization/V2/InitializeParticle(Color=<rgba 1 1 1 0>, Sprite Size Mode=::NewEnumerator3, Uniform Sprite Size=1, UsePositionOffset=false, Write Material Random=true)
   &SystemLocation(MaskSpawnGroup=false, Offset=<0 0 50>) {enabled=false}
  ]
  particleUpdateScript [
   &ParticleState(Kill Particles When Lifetime Has Elapsed=false, Let Infinitely Lived Particles Die When Emitter Deactivates=true)
   /SetVariables(Particles.Position=&Add_Vector(A=Linked() {linkedVariable {name=Engine.Owner.Position, type=NiagaraPosition}}, B=<0 0 60>)) {isSetParameters=true, module=SetVariables_#2}
   &TorusLocation(
    Align Torus With Axis=false
    Handle Radius=0
    Large Radius=80
    MaskSpawnGroup=false
    Radius Position=0
    Torus Distribution Mode:ENiagaraTorusDistributionMode=::NewEnumerator1
    U Position:NiagaraFloat=&Divide_Float(A=&MakeFloatFromInt(Int=&ReturnParticleIDAttributes()), B=7)
   )
  ]
  rendererClasses [&NiagaraSpriteRendererProperties]
  renderers [{rendererClass=&NiagaraSpriteRendererProperties}]
 }
 ParentDetachable {
  emitterUpdateScript [
   &EmitterState()
   &SpawnBurst_Instantaneous(Spawn Count=7, Use Spawn Probability=false) {enabled=false}
   &SpawnRate(SpawnRate=5, Use Spawn Probability=false)
  ]
  particleSpawnScript [
   &InitializeParticle(Color=<rgba 1 1 1 0>, Sprite Size Mode=::NewEnumerator3, Uniform Sprite Size=1, UsePositionOffset=false, Write Material Random=true)
   &SystemLocation(MaskSpawnGroup=false, Offset=<0 0 50>) {enabled=false}
  ]
  particleUpdateScript [
   &ParticleState()
   /SetVariables(Particles.Position=&Add_Vector(A=Linked() {linkedVariable {name=Engine.Owner.Position, type=NiagaraPosition}}, B=<0 0 60>)) {isSetParameters=true, module=SetVariables_#2}
   &TorusLocation(
    Align Torus With Axis=false
    Handle Radius=0
    Large Radius=80
    MaskSpawnGroup=false
    Radius Position=0
    Torus Distribution Mode:ENiagaraTorusDistributionMode=::NewEnumerator1
    U Position:NiagaraFloat=&Divide_Float(A=&MakeFloatFromInt(Int=&ReturnParticleIDAttributes()), B=5)
   )
  ]
  rendererClasses [&NiagaraSpriteRendererProperties]
  renderers [{rendererClass=&NiagaraSpriteRendererProperties}]
 }
 Smoke {
  emitterUpdateScript [
   &EmitterState()
   &SpawnRate(SpawnRate=21, Use Spawn Probability=false)
  ]
  particleSpawnScript [
   &InitializeParticle(
    Color=<rgba 5 5 5 0.2>
    Lifetime=1.5
    Sprite Size Max=<30 50>
    Sprite Size Min=<15 30>
    Sprite Size Mode=::NewEnumerator2
    UsePositionOffset=false
    Write Material Random=true
   )
   /SetVariables(Particles.Position=&Add_Vector(A=Linked() {linkedVariable {name=Engine.Owner.Position, type=NiagaraPosition}}, B=<0 0 60>)) {isSetParameters=true, module=SetVariables_#2}
   &TorusLocation(Align Torus With Axis=false, Handle Radius=0, MaskSpawnGroup=false, Module.Override Randomness=false)
   /SetVariables(Particles.RandomFrame=&RandomRangeInt(Minimum=0, Maximum=63)) {isSetParameters=true, module=SetVariables_#3}
   &AddVelocity(Velocity=&MakeVector(X=0, Y=0, Z=&MakeCustomFloatFromBool(Bool=$User.bIsFrontEndPreview, True Float=250, False Float=0)))
  ]
  particleUpdateScript [
   &ParticleState()
   &InheritSourceMovement(
    Applied Position Scale=&VectorFromFloat(Value=&FloatFromCurve(FloatCurve=DataInterface() {propertyValues {Curve {keys [<at 0 1>,<at 1 0.3>]}}}))
    ApplyForce=false
    ApplyPosition=true
    ApplyPositionFalloff=false
   )
   &SolveForcesAndVelocity(Clamp Velocity=false, Limit Acceleration=false)
   &ScaleColor(
    Scale Alpha=&FloatFromCurve(FloatCurve=DataInterface() {propertyValues {Curve {keys [
         {arriveTangent=2.9808, interpMode=RCIM_Cubic, leaveTangent=2.9808, tangentMode=RCTM_User, time=0, value=0}
         <at 0.5 1 RCIM_Cubic>
         {arriveTangent=-1.6379, interpMode=RCIM_Cubic, leaveTangent=-1.6379, tangentMode=RCTM_User, time=1, value=0}
        ]}}})
    Scale RGB=&MakeVectorFromLinearColorRGB(Color=&ColorFromCurve(ColorCurve=DataInterface() {propertyValues {
        AlphaCurve {keys [<at 0 1 RCIM_Cubic>]}
        BlueCurve {keys [<at 0.0463 3>,<at 0.2948 1>,<at 0.6594 1 RCIM_Cubic>]}
        GreenCurve {keys [<at 0.0463 1.7348>,<at 0.2948 0.1102>,<at 0.6594 0 RCIM_Cubic>]}
        RedCurve {keys [<at 0.0463 0>,<at 0.2948 0>,<at 0.6594 0.4798 RCIM_Cubic>]}
       }}))
    ScaleA=true
    ScaleRGB=true
   )
   &ScaleSpriteSize(Scale Factor=&Vector2DFromFloat(Value=&FloatFromCurve(
      FloatCurve=DataInterface() {propertyValues {Curve {keys [{arriveTangent=2.4681, interpMode=RCIM_Cubic, leaveTangent=2.4681, tangentMode=RCTM_User, time=0, value=0}, <at 0.5 1>, <at 1 0 RCIM_Cubic>]}}}
      Scale Curve=5
     )))
   &FN_SubUVAnimation(End Frame=&Add_Integer(A=$Particles.RandomFrame, B=63), Start Frame=$Particles.RandomFrame)
  ]
  rendererClasses [&NiagaraSpriteRendererProperties]
  renderers [{rendererClass=&NiagaraSpriteRendererProperties}]
 }
 Smoke_Center {
  emitterUpdateScript [
   &EmitterState()
   &SpawnRate(SpawnRate=50, Use Spawn Probability=false)
  ]
  particleSpawnScript [
   &InitializeParticle(
    Color=<rgba 3 3 3 0.75>
    Lifetime=&RandomRangeFloat(Minimum=0.4, Maximum=0.6)
    Sprite Size Max=<30 50>
    Sprite Size Min=<15 30>
    Sprite Size Mode=::NewEnumerator2
    UsePositionOffset=false
    Write Material Random=true
   )
   /SetVariables(Particles.Position=&Add_Vector(A=Linked() {linkedVariable {name=Engine.Owner.Position, type=NiagaraPosition}}, B=<0 0 27>)) {isSetParameters=true, module=SetVariables_#2}
   &TorusLocation(Align Torus With Axis=false, Handle Radius=0, Large Radius=20, MaskSpawnGroup=false, Module.Override Randomness=false)
   /SetVariables(Particles.RandomFrame=&RandomRangeInt(Minimum=0, Maximum=63)) {isSetParameters=true, module=SetVariables_#3}
  ]
  particleUpdateScript [
   &ParticleState()
   &InheritSourceMovement(
    Applied Force Scale=<-0.5 -0.5 0>
    ApplyForce=true
    ApplyForceFalloff=false
    ApplyPosition=true
    ApplyPositionFalloff=false
    Module.LimitForceSpeed=false
   ) {module=InheritSourceMovement001}
   &AddVelocityFromPoint(
    Module.Use Falloff Distance=false
    Origin Offset=<0 0 22>
    Velocity Strength=&FloatFromCurve(
     FloatCurve=DataInterface() {propertyValues {Curve {keys [
         {arriveTangent=-1.0543, interpMode=RCIM_Cubic, time=0.25, value=1}
         {arriveTangent=-5, interpMode=RCIM_Cubic, time=0.6, value=0}
         <at 0.7 -1 RCIM_Cubic>
        ]}}}
     Scale Curve=13
    )
   ) {enabled=false}
   &PointForce(
    Force Falloff Distance=160
    Force Origin=&Add_Vector(A=/Niagara/DynamicInputs/Helpers/SimulationPosition(), B=<0 0 22>)
    Force Strength=1500
    Module.Use Falloff=true
    Module.UseLinearForce=false
   )
   &GravityForce(Gravity=&MakeVector(
     X=0
     Y=0
     Z=&FloatFromCurve(
      FloatCurve=DataInterface() {propertyValues {Curve {keys [<at 0.4 0 RCIM_Cubic>, {arriveTangent=0.5303, interpMode=RCIM_Cubic, leaveTangent=0.5303, tangentMode=RCTM_User, time=0.5, value=1}]}}}
      Scale Curve=3000
     )
    ))
   &SolveForcesAndVelocity(Clamp Velocity=false, Limit Acceleration=false)
   &ScaleColor(
    Scale Alpha=&FloatFromCurve(FloatCurve=DataInterface() {propertyValues {Curve {keys [
         {arriveTangent=2.9808, interpMode=RCIM_Cubic, leaveTangent=2.9808, tangentMode=RCTM_User, time=0, value=0}
         <at 0.5 1 RCIM_Cubic>
         {arriveTangent=-1.6379, interpMode=RCIM_Cubic, leaveTangent=-1.6379, tangentMode=RCTM_User, time=1, value=0}
        ]}}})
    Scale RGB=&MakeVectorFromLinearColorRGB(Color=&ColorFromCurve(ColorCurve=DataInterface() {propertyValues {
        AlphaCurve {keys [<at 0 1 RCIM_Cubic>,<at 0.6641 0>]}
        BlueCurve {keys [<at 0.0463 5>,<at 0.5945 1>]}
        GreenCurve {keys [<at 0.0463 2.8913>,<at 0.5945 0.1102>]}
        RedCurve {keys [<at 0.0463 0>,<at 0.5945 0>]}
       }}))
    ScaleA=true
    ScaleRGB=true
   )
   &ScaleSpriteSize(Scale Factor=&Vector2DFromFloat(Value=&FloatFromCurve(
      FloatCurve=DataInterface() {propertyValues {Curve {keys [{arriveTangent=2.4681, interpMode=RCIM_Cubic, leaveTangent=2.4681, tangentMode=RCTM_User, time=0, value=0}, <at 1 1 RCIM_Cubic>]}}}
      Scale Curve=3.5
     )))
   &FN_SubUVAnimation(End Frame=&Add_Integer(A=$Particles.RandomFrame, B=24), Start Frame=$Particles.RandomFrame)
  ]
  rendererClasses [&NiagaraSpriteRendererProperties]
  renderers [{rendererClass=&NiagaraSpriteRendererProperties}]
 }
 SpeedLines {
  emitterUpdateScript [
   &EmitterState()
   &SpawnRate(SpawnRate=50, Use Spawn Probability=false)
  ]
  particleSpawnScript [
   &InitializeParticle(
    Color=<rgba 5 5 5 0.15>
    Lifetime=0.5
    Sprite Size Max=<1.25 8>
    Sprite Size Min=<0.5 5>
    Sprite Size Mode=::NewEnumerator2
    UsePositionOffset=false
    Write Material Random=true
   )
   /SetVariables(Particles.Position=&Add_Vector(A=Linked() {linkedVariable {name=Engine.Owner.Position, type=NiagaraPosition}}, B=<0 0 22>)) {isSetParameters=true, module=SetVariables_#2}
   &TorusLocation(Align Torus With Axis=false, Handle Radius=0, Large Radius=20, MaskSpawnGroup=false, Module.Override Randomness=false)
  ]
  particleUpdateScript [
   &ParticleState()
   &InheritSourceMovement(
    Applied Force Scale=<-0.5 -0.5 0>
    ApplyForce=true
    ApplyForceFalloff=false
    ApplyPosition=true
    ApplyPositionFalloff=false
    Module.LimitForceSpeed=false
   )
   &AddVelocityFromPoint(
    Module.Use Falloff Distance=false
    Origin Offset=<0 0 22>
    Velocity Strength=&FloatFromCurve(
     FloatCurve=DataInterface() {propertyValues {Curve {keys [
         {arriveTangent=-1.0543, interpMode=RCIM_Cubic, time=0.25, value=1}
         {arriveTangent=-5, interpMode=RCIM_Cubic, time=0.6, value=0}
         <at 0.7 -1 RCIM_Cubic>
        ]}}}
     Scale Curve=13
    )
   ) {enabled=false}
   &PointForce(
    Force Falloff Distance=160
    Force Origin=&Add_Vector(A=&SimulationPosition(), B=<0 0 20>)
    Force Strength=1500
    Module.Use Falloff=true
    Module.UseLinearForce=false
   )
   &GravityForce(Gravity=&MakeVector(
     X=0
     Y=0
     Z=&FloatFromCurve(
      FloatCurve=DataInterface() {propertyValues {Curve {keys [<at 0.4 0 RCIM_Cubic>, {arriveTangent=0.5303, interpMode=RCIM_Cubic, leaveTangent=0.5303, tangentMode=RCTM_User, time=0.5, value=1}]}}}
      Scale Curve=3000
     )
    ))
   &SolveForcesAndVelocity(Clamp Velocity=false, Limit Acceleration=false)
   &ScaleColor(
    Scale Alpha=&FloatFromCurve(FloatCurve=DataInterface() {propertyValues {Curve {keys [
         {arriveTangent=2.9808, interpMode=RCIM_Cubic, leaveTangent=2.9808, tangentMode=RCTM_User, time=0, value=0}
         <at 0.5 1 RCIM_Cubic>
         {arriveTangent=-1.6379, interpMode=RCIM_Cubic, leaveTangent=-1.6379, tangentMode=RCTM_User, time=1, value=0}
        ]}}})
    Scale RGB=&MakeVectorFromLinearColorRGB(Color=&ColorFromCurve(ColorCurve=DataInterface() {propertyValues {
        AlphaCurve {keys [<at 0 1 RCIM_Cubic>,<at 0.6502 0>]}
        BlueCurve {keys [<at 0.0463 1>,<at 0.6351 1>]}
        GreenCurve {keys [<at 0.0463 0.889>,<at 0.6351 0>]}
        RedCurve {keys [<at 0.0463 0>,<at 0.6351 0.1722>]}
       }}))
    ScaleA=true
    ScaleRGB=true
   )
   &ScaleSpriteSize(Scale Factor=&Vector2DFromFloat(Value=&FloatFromCurve(
      FloatCurve=DataInterface() {propertyValues {Curve {keys [{arriveTangent=2.4681, interpMode=RCIM_Cubic, leaveTangent=2.4681, tangentMode=RCTM_User, time=0, value=0}, <at 0.5 1>, <at 1 1 RCIM_Cubic>]}}}
      Scale Curve=5
     )))
  ]
  rendererClasses [&NiagaraSpriteRendererProperties]
  renderers [{rendererClass=&NiagaraSpriteRendererProperties}]
 }
 Trail_Child001 {
  emitterUpdateScript [
   &EmitterState()
   /SetVariables(Emitter.EmitterReader=DataInterface() {propertyValues {EmitterBinding {bindingMode=Other, emitterName=Parent}}}) {isSetParameters=true, module=SetVariables_#4}
   &SpawnRate(
    SpawnRate=&self:Spawn_by_Emitter_01(New Particle Attribute Reader:NiagaraDataInterfaceParticleRead=$Emitter.EmitterReader, Rate:NiagaraFloat=30)
    Use Spawn Probability=true
   )
  ]
  particleSpawnScript [
   &self:Sample_Location_01(New Particle Attribute Reader:NiagaraDataInterfaceParticleRead=$Emitter.EmitterReader) {module=Sample_Location_01}
   &InitializeRibbon(
    Lifetime=0.7
    Module.Write Color=true
    Module.Write Lifetime=true
    Module.Write Mass=true
    Module.Write Material Random=true
    Module.Write Position=false
    Module.Write RibbonTwist=false
    Module.Write RibbonWidth=true
    Ribbon Width=15
   )
   /SetVariables(Particles.MaterialRandom:NiagaraFloat=&RandomRangeFloat(Minimum=0, Maximum=0.2)) {isSetParameters=true, module=SetVariables_#5}
  ]
  particleUpdateScript [
   &ParticleState()
   &ScaleRibbonWidth(Ribbon Width Scale=&FloatFromCurve(FloatCurve=DataInterface() {propertyValues {Curve {keys [{arriveTangent=3.5344, interpMode=RCIM_Cubic, leaveTangent=3.5344, tangentMode=RCTM_User, time=0, value=0}, <at 0.5 1>]}}}))
   &AddVelocityFromPoint(Module.Use Falloff Distance=true, Origin Offset=<0 0 100>, Velocity Falloff Distance=105, Velocity Strength=200) {enabled=false, module=AddVelocityFromPoint001}
   &PointForce(
    Force Falloff Distance=200
    Force Origin=&Add_Vector(A=&SimulationPosition(), B=<0 0 100>)
    Force Strength=1000
    Module.Use Falloff=true
    Module.UseLinearForce=false
   )
   &LinearForce(
    Coordinate Space:ENiagaraCoordinateSpace=::World
    Force=&MakeVector(
     X=0
     Y=0
     Z=&FloatFromCurve(
      FloatCurve=DataInterface() {propertyValues {Curve {keys [<at 0.25 0.2>,<at 0.302 1>,<at 1 1>]}}}
      Scale Curve=&MakeCustomFloatFromBool(Bool=$User.bIsFrontEndPreview, True Float=2000, False Float=0)
     )
    )
   )
   &InheritSourceMovement(
    Applied Position Scale=&VectorFromFloat(Value=&FloatFromCurve(FloatCurve=DataInterface() {propertyValues {Curve {keys [<at 0 1>,<at 1 0>]}}}))
    ApplyForce=false
    ApplyPosition=true
    ApplyPositionFalloff=false
   )
   &SolveForcesAndVelocity(Clamp Velocity=false, Limit Acceleration=false)
   &FN_RibbonMaterial(
    Color=&ColorFromCurve(ColorCurve=DataInterface() {propertyValues {
       AlphaCurve {keys [<at 0 1 RCIM_Cubic>]}
       BlueCurve {keys [<at -0.0028 1>,<at 0.4954 2>,<at 0.947 1 RCIM_Cubic>]}
       GreenCurve {keys [<at -0.0028 0>,<at 0.4954 1.1565>,<at 0.947 0 RCIM_Cubic>]}
       RedCurve {keys [<at -0.0028 0.2326>,<at 0.4954 0>,<at 0.947 0.2326 RCIM_Cubic>]}
      }})
    Color Pan U Offset:NiagaraFloat=$Particles.Initial.MaterialRandom
    Color Pan U Speed=1
    Color Scale U=0.5
    Erosion Pan Offset U:NiagaraFloat=$Particles.Initial.MaterialRandom
    Erosion Pan Speed U=1
    Erossion Scale U=0.5
    Macro Pan Offset:NiagaraFloat=$Particles.Initial.MaterialRandom
    Macro Scale U=0.5
    Opacity Pan U Offset:NiagaraFloat=$Particles.Initial.MaterialRandom
    Opacity Pan U Speed=1
    Opacity Scale U=0.5
    Scale Alpha:NiagaraFloat=&FloatFromCurve(
     FloatCurve=DataInterface() {propertyValues {Curve {keys [
         {arriveTangent=-0.3763, interpMode=RCIM_Cubic, leaveTangent=-0.3763, tangentMode=RCTM_User, time=0, value=0}
         <at 0.5 1 RCIM_Cubic>
         {arriveTangent=-1.1615, interpMode=RCIM_Cubic, leaveTangent=-1.1615, tangentMode=RCTM_User, time=1, value=0}
        ]}}}
     Scale Curve=0.5
    )
   )
  ]
  rendererClasses [&NiagaraRibbonRendererProperties]
  renderers [{rendererClass=&NiagaraRibbonRendererProperties}]
 }
 Trail_ChildDetachable {
  emitterUpdateScript [
   &EmitterState()
   /SetVariables(Emitter.EmitterReader=DataInterface() {propertyValues {EmitterBinding {bindingMode=Other, emitterName=ParentDetachable}}}) {isSetParameters=true, module=SetVariables_#4}
   &SpawnRate(
    SpawnRate=&self:Spawn_by_Emitter_01(New Particle Attribute Reader:NiagaraDataInterfaceParticleRead=$Emitter.EmitterReader, Rate:NiagaraFloat=16)
    Use Spawn Probability=true
   )
  ]
  particleSpawnScript [
   &self:Sample_Location_01(New Particle Attribute Reader:NiagaraDataInterfaceParticleRead=$Emitter.EmitterReader) {module=Sample_Location_01}
   &InitializeRibbon(
    Lifetime=1.5
    Module.Write Color=true
    Module.Write Lifetime=true
    Module.Write Mass=true
    Module.Write Material Random=true
    Module.Write Position=false
    Module.Write RibbonTwist=false
    Module.Write RibbonWidth=true
    Ribbon Width=35
   )
   /SetVariables(Particles.MaterialRandom:NiagaraFloat=&RandomRangeFloat(Minimum=0, Maximum=0.2)) {isSetParameters=true, module=SetVariables_#5}
  ]
  particleUpdateScript [
   &ParticleState()
   &ScaleRibbonWidth(Ribbon Width Scale=&FloatFromCurve(FloatCurve=DataInterface() {propertyValues {Curve {keys [
         {arriveTangent=13.755, interpMode=RCIM_Cubic, leaveTangent=6.6359, tangentMode=RCTM_User, time=0, value=0}
         <at 0.2127 1 RCIM_Cubic>
         {arriveTangent=-5.1866, interpMode=RCIM_Cubic, leaveTangent=-5.1866, tangentMode=RCTM_User, time=0.5, value=0}
        ]}}}))
   &AddVelocityFromPoint(Module.Use Falloff Distance=true, Origin Offset=<0 0 100>, Velocity Falloff Distance=105, Velocity Strength=200) {enabled=false, module=AddVelocityFromPoint001}
   &PointForce(
    Force Falloff Distance=150
    Force Origin=&Add_Vector(A=&SimulationPosition(), B=<0 0 100>)
    Force Strength=1000
    Module.Use Falloff=true
    Module.UseLinearForce=false
   )
   &LinearForce(
    Coordinate Space:ENiagaraCoordinateSpace=::World
    Force=&MakeVector(
     X=0
     Y=0
     Z=&FloatFromCurve(
      FloatCurve=DataInterface() {propertyValues {Curve {keys [<at 0.25 0.2>,<at 0.302 1>,<at 1 1>]}}}
      Scale Curve=&MakeCustomFloatFromBool(Bool=$User.bIsFrontEndPreview, True Float=2000, False Float=0)
     )
    )
   )
   &InheritSourceMovement(
    Applied Position Scale=&VectorFromFloat(Value=&FloatFromCurve(FloatCurve=DataInterface() {propertyValues {Curve {keys [<at 0 1>, {leaveTangent=-0.5156, time=1, value=0}]}}}))
    ApplyForce=false
    ApplyPosition=true
    ApplyPositionFalloff=false
   )
   &SolveForcesAndVelocity(Clamp Velocity=false, Limit Acceleration=false)
   &FN_RibbonMaterial(
    Color=&ColorFromCurve(ColorCurve=DataInterface() {propertyValues {
       AlphaCurve {keys [<at 0 1 RCIM_Cubic>]}
       BlueCurve {keys [<at 0.1134 1>,<at 0.4564 2>,<at 0.7158 1 RCIM_Cubic>]}
       GreenCurve {keys [<at 0.1134 1>,<at 0.4564 1.1565>,<at 0.7158 0 RCIM_Cubic>]}
       RedCurve {keys [<at 0.1134 1>,<at 0.4564 0>,<at 0.7158 0.7359 RCIM_Cubic>]}
      }})
    Color Pan U Offset:NiagaraFloat=$Particles.Initial.MaterialRandom
    Color Pan U Speed=1
    Color Scale U=0.5
    Erosion Pan Offset U:NiagaraFloat=$Particles.Initial.MaterialRandom
    Erosion Pan Speed U=1
    Erossion Scale U=0.5
    Macro Pan Offset:NiagaraFloat=$Particles.Initial.MaterialRandom
    Macro Scale U=0.5
    Opacity Pan U Offset:NiagaraFloat=$Particles.Initial.MaterialRandom
    Opacity Pan U Speed=1
    Opacity Scale U=0.5
    Scale Alpha:NiagaraFloat=&FloatFromCurve(
     FloatCurve=DataInterface() {propertyValues {Curve {keys [
         {arriveTangent=-0.3763, interpMode=RCIM_Cubic, leaveTangent=-0.3763, tangentMode=RCTM_User, time=0, value=0}
         <at 0.5 1 RCIM_Cubic>
         {arriveTangent=-1.1615, interpMode=RCIM_Cubic, leaveTangent=-1.1615, tangentMode=RCTM_User, time=1, value=0}
        ]}}}
     Scale Curve=0.6
    )
   )
  ]
  rendererClasses [&NiagaraRibbonRendererProperties]
  renderers [{rendererClass=&NiagaraRibbonRendererProperties}]
 }
 Tube {
  emitterSpawnScript [&FN_PreserveSkydiveUserVariables() {module=PreserveSkydiveUserVariables}]
  emitterUpdateScript [
   &EmitterState(
    Life Cycle Mode=::NewEnumerator1
    Loop Behavior=::NewEnumerator1
    Loop Duration Mode=::NewEnumerator1
    Reset Age On Awaken=true
    Scalability Mode=::NewEnumerator1
    UseLoopDelay=false
   )
   &SpawnBurst_Instantaneous(Spawn Count=1, Use Spawn Probability=false)
  ]
  particleSpawnScript [
   &FN_SetSpawnGroup(Number of Spawn Groups=1) {module=FN_SetSpawnGroup001}
   /Niagara/Modules/Spawn/Initialization/InitializeParticle(
    Color=<rgba 3 3 3 1>
    Lifetime=2
    Mesh Scale=<1 1 0.3>
    Module.Write Color=true
    Module.Write Lifetime=true
    Module.Write Mass=false
    Module.Write Material Random=true
    Module.Write Position=false
    Module.Write Scale=true
    Module.Write SpriteRotation=false
    Module.Write SpriteSize=false
    Position=/Niagara/Functions/Localspace/SimulationPosition()
    Sprite Rotation=&UniformRangedFloat(Minimum=0, Maximum=1)
   )
   &SkeletalMeshLocation(
    Bone / Socket Index=4
    Bone Sampling Mode=::NewEnumerator4
    Inherit Velocity=false
    Interpolate Mesh Sampling=false
    MaskSpawnGroup=false
    Sampled Position Offset=<0 0 50>
    Skeletal Mesh=$User.SkeletalMesh
   )
   /SetVariables(Particles.Renderer2VisibilityTag:NiagaraInt32=&MakeCustomIntFromBool(Bool=$User.bIsFrontEndPreview, True INT=0, False INT=1)) {isSetParameters=true}
  ]
  particleUpdateScript [
   &FN_DivingAttributes_v2(Accumulate Dive=0.1, Accumulate Rest=0.1, SkeletalMesh:NiagaraDataInterfaceSkeletalMesh=$User.SkeletalMesh) {module=FN_DivingAttributes_v2001}
   &ParticleState(Kill Particles When Lifetime Has Elapsed=false, Let Infinitely Lived Particles Die When Emitter Deactivates=true)
   &SolveForcesAndVelocity(Clamp Velocity=false, Limit Acceleration=false)
   &ScaleColor(
    Scale Alpha=&FloatFromCurve(FloatCurve=DataInterface() {propertyValues {Curve {keys [
         {arriveTangent=4.0516, interpMode=RCIM_Cubic, leaveTangent=4.0516, tangentMode=RCTM_User, time=0, value=0}
         {arriveTangent=0.0961, interpMode=RCIM_Cubic, leaveTangent=0.0961, tangentMode=RCTM_User, time=0.25, value=1}
         {arriveTangent=0.2272, interpMode=RCIM_Cubic, leaveTangent=0.2272, tangentMode=RCTM_User, time=1, value=1}
        ]}}})
    ScaleA=true
    ScaleRGB=false
   )
   &InterpolateOverTime(
    Rate Of Change=4
    Target Vector Value:Vector3f=/Niagara/DynamicInputs/Transforms/TransformPosition(Position=<0 0 0>, Source Space=::Local, Destination Space=::World)
    Variable Type=::NewEnumerator2
   )
   &DynamicMaterialParameters(
    Index 0 Param 1=&MakeFloatFromVector(Vector=$Output.InterpolateOverTime.Moving Average, Channel=::NewEnumerator0)
    Index 0 Param 2=&MakeFloatFromVector(Vector=$Output.InterpolateOverTime.Moving Average, Channel=::NewEnumerator1)
    Index 0 Param 3=&MakeFloatFromVector(
     Vector=&MakeCustomVectorFromBool(Bool=$User.bIsFrontEndPreview, True Vector=<0 0 150>, False Vector=$Output.InterpolateOverTime.Moving Average)
     Channel=::NewEnumerator2
    )
    Index 0 Param 4=2
    Index 1 Param 1=&Lerp_Float(
     A=&MakeCustomFloatFromBool(Bool=$User.bIsFrontEndPreview, True Float=0.5, False Float=1)
     B=&MakeCustomFloatFromBool(Bool=$User.bIsFrontEndPreview, True Float=0.5, False Float=2)
     Alpha=$System.FN_Integrator001.Value
    )
    Index 1 Param 2=&Lerp_Float(
     A=1
     B=&MakeCustomFloatFromBool(Bool=$User.bIsFrontEndPreview, True Float=1.2, False Float=2)
     Alpha=$System.FN_Integrator001.Value
    )
    Module.Param0WriteEnabled=true
    Module.Param0WriteEnabled1=true
    Module.Param1WriteEnabled=true
    Module.Param1WriteEnabled1=true
    Module.Param2WriteEnabled=true
    Module.Param2WriteEnabled1=true
    Module.Param3WriteEnabled=true
    Module.Param3WriteEnabled1=true
    Write Parameter Index 1=true
   )
   &FN_Light_Attributes(
    Color=<rgba 1.2913 3.4976 6.5 1>
    Create Color Attribute=true
    Create Exponent Attribute=true
    Create Radius Attribute=true
    Exponent=5
    Radius=150
   )
  ]
  rendererClasses [&NiagaraMeshRendererProperties, &NiagaraLightRendererProperties]
  renderers [
   {rendererClass=&NiagaraMeshRendererProperties, rendererIndex=0}
   {rendererClass=&NiagaraMeshRendererProperties, rendererIndex=1}
   {rendererClass=&NiagaraLightRendererProperties, rendererIndex=2}
  ]
 }
}
summary {
 systemName=NS_Skydive_Tube
 emitters [
  {emitterName=Tube, rendererClasses [&NiagaraMeshRendererProperties, &NiagaraLightRendererProperties]}
  {emitterName=Parent, rendererClasses [&NiagaraSpriteRendererProperties]}
  {emitterName=ParentDetachable, rendererClasses [&NiagaraSpriteRendererProperties]}
  {emitterName=Trail_ChildDetachable, rendererClasses [&NiagaraRibbonRendererProperties]}
  {emitterName=Smoke, rendererClasses [&NiagaraSpriteRendererProperties]}
  {emitterName=Trail_Child001, rendererClasses [&NiagaraRibbonRendererProperties]}
  {emitterName=SpeedLines, rendererClasses [&NiagaraSpriteRendererProperties]}
  {emitterName=Smoke_Center, rendererClasses [&NiagaraSpriteRendererProperties]}
  {emitterName=Bowl_Multi, rendererClasses [&NiagaraMeshRendererProperties]}
  {emitterName=CenterGlow, rendererClasses [&NiagaraMeshRendererProperties]}
 ]
 userVariables [
  User.bIsFrontEnd {type=NiagaraBool}
  User.bIsFrontEndPreview {type=NiagaraBool}
  User.Color_Tint {type=LinearColor, description="", defaultValue {struct=LinearColor, value=<rgba 1 1 1 1>}}
  User.EffectIntensity {type=NiagaraFloat, defaultValue=0, description=""}
  User.ParaGlideLeanAlpha {type=NiagaraFloat, defaultValue=0, description=""}
  User.PlayerSpeed {type=NiagaraFloat, defaultValue=0, description=""}
  User.RibbonIDOffset {type=NiagaraInt32}
  User.SkeletalMesh {
   type=NiagaraDataInterfaceSkeletalMesh
   defaultValue {
    struct=NiagaraExt_VariableValue_DataInterface
    value {dataInterface=&self:NiagaraDataInterfaceSkeletalMesh_0}
   }
  }
 ]
}
